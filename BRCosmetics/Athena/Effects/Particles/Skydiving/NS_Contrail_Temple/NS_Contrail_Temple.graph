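# Niagara skydiving contrail cosmetic that puts two flickering flares, with trailing smoke and sparks, on the player's skeletal mesh.
# Each emitter splits into two spawn groups pinned to sampled bones or sockets: flare meshes with glow sprites and lights, short ribbons, GPU smoke and GPU sparks.
# A system-level integrator, apparently tracking dive, blends offsets, velocities and spawn rates.
# Time-of-day and front-end modifiers rebalance brightness and motion.
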NiagaraSystem /BRCosmetics/Athena/Effects/Particles/Skydiving/NS_Contrail_Temple/NS_Contrail_Temple
dependencies {
 usedDataInterfaces [
  &NiagaraDataInterfaceArrayFloat3
  &NiagaraDataInterfaceColorCurve
  &NiagaraDataInterfaceCurve
  &NiagaraDataInterfaceVectorCurve
  &NiagaraDataInterfaceVector2DCurve
 ]
 usedDynamicInputs []
 usedModules [
  &SystemState
  &self:SystemSpawnScript.SystemScriptSource.SystemScriptGraph.NiagaraNodeAssignment_0.SetVariables_#1
  &FN_Integrator
 ]
}
emitters {
 DistressFlare001 {
  emitterUpdateScript [
   &EmitterState()
   &SpawnBurst_Instantaneous(Spawn Count=2, Spawn Probability=HlslExpression() {hlslExpression="Emitter.LoopCount == 0"}, Use Spawn Probability=true)
  ]
  particleSpawnScript [
   &FN_SetSpawnGroup(Number of Spawn Groups=2)
   /Niagara/Modules/Spawn/Initialization/InitializeParticle(
    Mesh Scale=<0.05 0.05 0.05>
    Module.Write Color=true
    Module.Write Lifetime=true
    Module.Write Mass=true
    Module.Write Material Random=true
    Module.Write Position=false
    Module.Write Scale=true
    Module.Write SpriteRotation=false
    Module.Write SpriteSize=true
    Position=&SimulationPosition()
    Sprite Size=<10 40>
   )
   &InitialMeshOrientation(
    Orientation Axis=<0 0 0>
    Orientation Vector=<0 0 0>
    Rotation=&Lerp_Vector(A=<0 0 0>, B=<0 0.5 0>, Alpha=&MakeFloatFromInt(Int=&SpawnGroup()))
    Rotation Coordinate Space=::NewEnumerator0
    Use Orientation Vector=true
    Use Rotation Vector=true
   )
   &SkeletalMeshLocation(
    Apply Sampled Orientation to Mesh Particles=true
    Bone / Socket Index=$Particles.SpawnGroup
    Inherit Velocity=false
    Interpolate Mesh Sampling=false
    MaskSpawnGroup=false
    Mesh Sampling Type=::NewEnumerator5
    Sampled Position Offset=&SelectVectorFromArray(
     Direct Array Index=&SpawnGroup()
     Vector Selection Array=DataInterface() {propertyValues {FloatData [<0 -2 5>, <0 2 -5>]}}
    )
    Sampled Position Offset Coordinate Space:ENiagaraMeshTransforms=::NewEnumerator3
    Skeletal Mesh=$User.SkeletalMesh
    Socket Sampling Mode=::NewEnumerator5
   ) {module=SkeletalMeshLocation001}
   &FN_Mesh_Attributes(
    Create Orientation Attribute=true
    Create Position Attribute=true
    Create Scale Attribute=true
    Position Offset:Vector3f=<40 0 0>
    Scale=<0.15 0.15 0.4>
    Write to Module Namespace=true
   )
   /SetVariables(Particles.GnomonScale:Vector3f=<5 5 5>) {isSetParameters=true}
   /SetVariables() {isSetParameters=true}
  ]
  particleUpdateScript [
   &ParticleState(Kill Particles When Lifetime Has Elapsed=false)
   &SkeletalMeshLocation(
    Apply Sampled Orientation to Mesh Particles=true
    Bone / Socket Index=$Particles.SpawnGroup
    Inherit Velocity=false
    Interpolate Mesh Sampling=false
    MaskSpawnGroup=false
    Mesh Sampling Type=::NewEnumerator5
    Sampled Position Offset=&SelectVectorFromArray(
     Direct Array Index=&SpawnGroup()
     Vector Selection Array=DataInterface() {propertyValues {FloatData [<0 -2 5>, <0 2 -5>]}}
    )
    Sampled Position Offset Coordinate Space:ENiagaraMeshTransforms=::NewEnumerator3
    Skeletal Mesh=$User.SkeletalMesh
    Socket Sampling Mode=::NewEnumerator5
   )
   &Color(
    Color=&MultiplyLinearColor(
     Linear Color=&LerpLinearColors(
      StartColor=<rgba 40000 3570.5566 0 1>
      EndColor=<rgba 10000 389.5509 0 1>
      LerpFactor=&MakeCustomFloatFromBool(Bool=$User.bIsFrontEndPreview, True Float=1, False Float=0)
     )
     Scale Factor=<1 1 1 1>
    )
    Scale Alpha=&FNTimeOfDayScalar(Night Scale=0.1, Morning Scale=0.5, Day Scale=0.7, Evening Scale=0.1)
   )
   &FN_Sprite_Attributes(Create Position Attribute=true, Position:Vector3f=$Output.SkeletalMeshLocation.SampledPosition, Write to Module Namespace=true) {module=FN_Sprite_Attributes001}
   /SetVariables(Particles.FN_Mesh_Attributes.Position:Vector3f=&Add_Vector(
     A=Linked() {linkedVariable {name=Particles.Position, type=NiagaraPosition}}
     B=&RotateVectorByQuaternion(
      Vector To Rotate=&Lerp_Vector(A=<-1 0 15>, B=<0 0 -15>, Alpha=&MakeFloatFromInt(Int=$Particles.SpawnGroup))
      Quaternion=$Output.SkeletalMeshLocation.SampledRotation
     )
    )) {isSetParameters=true}
   &FN_Sprite_Attributes(
    Color=&MultiplyLinearColor(
     Linear Color=<rgba 0 54.9466 200 1>
     Scale Factor=&MakeVector4_FromFloat(Float=&FNTimeOfDayScalar(Night Scale=0.4, Morning Scale=1, Day Scale=0.7, Evening Scale=0.7))
    )
    Create Color Attribute=true
    Create Position Attribute=true
    Create SpriteSize Attribute=true
    Position:Vector3f=&Add_Vector(
     A=$Particles.FN_Mesh_Attributes.Position
     B=&RotateVectorByQuaternion(
      Vector To Rotate=&Lerp_Vector(A=<0 0 0>, B=<0 0 0>, Alpha=&MakeFloatFromInt(Int=$Particles.SpawnGroup))
      Quaternion=$Output.SkeletalMeshLocation.SampledRotation
     )
    )
    SpriteSize=<140 15>
    Write to Module Namespace=true
   ) {module=FN_Sprite_Attributes002}
   &FN_Sprite_Attributes(
    Color=&MultiplyLinearColor(
     Linear Color=<rgba 4 0.0824 0.162 1>
     Scale Factor=&MakeVector4_FromFloat(Float=&FNTimeOfDayScalar(Night Scale=0.1, Morning Scale=0.3, Day Scale=0.3, Evening Scale=0.1))
    )
    Create Color Attribute=true
    Create Position Attribute=true
    Create SpriteSize Attribute=true
    Position:Vector3f=&Add_Vector(
     A=Linked() {linkedVariable {name=Particles.Position, type=NiagaraPosition}}
     B=&RotateVectorByQuaternion(
      Vector To Rotate=&Lerp_Vector(A=<0 0 0>, B=<0 2 0>, Alpha=&MakeFloatFromInt(Int=$Particles.SpawnGroup))
      Quaternion=$Output.SkeletalMeshLocation.SampledRotation
     )
    )
    SpriteSize=<100 100>
    Write to Module Namespace=true
   ) {module=FN_Sprite_Attributes003}
   /SetVariables(Particles.FN_Sprite_Attributes002.Color:LinearColor=&MultiplyLinearColor(
     Linear Color=$Particles.FN_Sprite_Attributes002.Color
     Scale Factor=&MakeFloat4FromVectorAndFloat(
      "Vector (RGB)"=&VectorFromFloat(Value=&Sine(Normalized Angle=$Emitter.Age, Period=0.6, Scale=0.2, Bias=0.4))
      "Float (Alpha)"=&Multiply_Float(
       A=&Multiply_Float(
        A=&FNTimeOfDayScalar(Night Scale=0.2, Morning Scale=0.5, Evening Scale=0.5)
        B=&FN_FrontEndFloatModifier(Gameplay=0.5, Front End=1)
       )
       B=&Lerp_Float(A=1, B=0.5, Alpha=$System.FN_Integrator001.NormalizedValue)
      )
     )
    )) {isSetParameters=true}
   /SetVariables(Particles.Flare_LightColor:LinearColor=&MultiplyLinearColor(
     Linear Color=<rgba 15000 720.0927 0 1>
     Scale Factor=&MakeFloat4FromVectorAndFloat(
      "Vector (RGB)"=&VectorFromFloat(Value=&Lerp_Float(
        A=&Multiply_Float(
         A=&Sine(Normalized Angle=$Emitter.Age, Period=&RandomRangeFloat(Minimum=0.22, Maximum=0.22), Scale=0.2, Bias=1)
         B=&Sine(
          Normalized Angle=$Emitter.Age
          Period=&RandomRangeFloat(Minimum=0.33, Maximum=0.33)
          Scale=0.2
          Bias=0.5
         )
        )
        B=&Multiply_Float(
         A=&Sine(Normalized Angle=$Emitter.Age, Period=&RandomRangeFloat(Minimum=0.2, Maximum=0.2), Scale=0.2, Bias=1)
         B=&Sine(Normalized Angle=$Emitter.Age, Period=&RandomRangeFloat(Minimum=0.3, Maximum=0.3), Scale=0.2, Bias=0.5)
        )
        Alpha=&MakeFloatFromInt(Int=$Particles.SpawnGroup)
       ))
      "Float (Alpha)"=&Multiply_Float(
       A=&Multiply_Float(
        A=&FNTimeOfDayScalar(Night Scale=0.2, Morning Scale=0.5, Evening Scale=0.5)
        B=&FN_FrontEndFloatModifier(Gameplay=0.5, Front End=1)
       )
       B=&Lerp_Float(A=1, B=0.5, Alpha=$System.FN_Integrator001.NormalizedValue)
      )
     )
    )) {isSetParameters=true}
  ]
  rendererClasses [
   &NiagaraMeshRendererProperties
   &NiagaraSpriteRendererProperties
   &NiagaraLightRendererProperties
  ]
  renderers [
   {rendererClass=&NiagaraMeshRendererProperties, rendererIndex=0}
   {rendererClass=&NiagaraMeshRendererProperties, rendererIndex=1}
   {rendererClass=&NiagaraSpriteRendererProperties, rendererIndex=2}
   {rendererClass=&NiagaraSpriteRendererProperties, rendererIndex=3}
   {rendererClass=&NiagaraLightRendererProperties, rendererIndex=4}
  ]
 }
 FlareRibbon001 {
  emitterUpdateScript [
   &EmitterState(Life Cycle Mode=::NewEnumerator1, Loop Duration=0, UseLoopDelay=false)
   /Game/Effects/Niagara/Modules/Skydiving/FN_DivingAttributes/FN_EmitterDivingAttributes_v1(Bone Index=2, SkeletalMesh:NiagaraDataInterfaceSkeletalMesh=$User.SkeletalMesh)
   &FN_SpawnRibbon_v2(Number of Ribbons=2, SpawnRate=40)
  ]
  particleSpawnScript [
   &FN_SetSpawnGroup_v2(Number of Spawn Groups=2)
   &FN_InitializeRibbon_v3(
    Color:LinearColor=<rgba 1 1 1 0>
    Enable Skeletal Position=true
    Lifetime=0.3
    Offset Amount:NiagaraFloat=&Lerp_Float(A=0, B=0, Alpha=&MakeFloatFromInt(Int=$Particles.SpawnGroup))
    Offset Axis:ENiagaraOrientationAxis=::XAxis
    Ribbon Width=70
    Write Material Random=false
   )
   &SkeletalMeshLocation(
    Apply Sampled Orientation to Mesh Particles=true
    Bone / Socket Index=$Particles.SpawnGroup
    Bone Sampling Mode=::NewEnumerator4
    Inherit Velocity=false
    Interpolate Mesh Sampling=false
    MaskSpawnGroup=false
    Sampled Position Offset=&SelectVectorFromArray(
     Direct Array Index=&SpawnGroup()
     Vector Selection Array=DataInterface() {propertyValues {FloatData [<0 0 0>, <0 0 0>]}}
    )
    Skeletal Mesh=$User.SkeletalMesh
   )
   /SetVariables(Particles.Position=&Add_Vector(
     A=Linked() {linkedVariable {name=Particles.Position, type=NiagaraPosition}}
     B=&RotateVectorByQuaternion(
      Vector To Rotate=&Lerp_Vector(
       A=&Lerp_Vector(A=<0 -10 20>, B=<0 5 -20>, Alpha=&MakeFloatFromInt(Int=$Particles.SpawnGroup))
       B=&FN_FrontEndVector3Modifier(
        Gameplay=&Lerp_Vector(A=<-2 0 25>, B=<-7 0 -25>, Alpha=&MakeFloatFromInt(Int=$Particles.SpawnGroup))
        Front End=&Lerp_Vector(A=<0 0 0>, B=<0 -5 -10>, Alpha=&MakeFloatFromInt(Int=$Particles.SpawnGroup))
       )
       Alpha=$System.FN_Integrator001.NormalizedValue
      )
      Quaternion=$Output.SkeletalMeshLocation.SampledRotation
     )
    )) {isSetParameters=true}
   &SphereLocation(
    MaskSpawnGroup=false
    Sphere Distribution:ENiagaraSphereDistributionMode=::NewEnumerator1
    Sphere Radius=0
    UseNonUniformScale=false
    UseOffset=true
    UseSphereOrientation=false
   )
   &AddVelocity(
    Scale Added Velocity=&Lerp_Vector(A=<1 2 1>, B=<1 -3 1>, Alpha=$System.FN_Integrator001.NormalizedValue)
    Velocity=&FN_FrontEndVector3Modifier(Gameplay=<0 150 900>, Front End=<0 0 900>)
   )
   &DynamicMaterialParameters(
    Index 0 Param 1=0.5
    Index 0 Param 2=&MaskFloatBySpawnGroup(Masked Float=0, Passthrough Float=0.2, Spawn Group Mask=1)
    Index 0 Param 3=1
    Module.Param0WriteEnabled=true
    Module.Param1WriteEnabled=true
    Module.Param2WriteEnabled=true
    Module.Param3WriteEnabled=false
   )
  ]
  particleUpdateScript [
   &InheritSourceMovement(Applied Position Scale=<0.7 0.7 1>, ApplyForce=false, ApplyPosition=true, ApplyPositionFalloff=false)
   /Game/Effects/Niagara/Modules/Skydiving/FN_DivingAttributes/v5/FN_DivingAttributes(Left Calf Index=10, SkeletalMesh:NiagaraDataInterfaceSkeletalMesh=$User.SkeletalMesh)
   &SolveForcesAndVelocity(Clamp Velocity=false, Limit Acceleration=false)
   &FN_UpdateRibbon(Enable Skeletal Position=true, Initial RibbonWidth:NiagaraFloat=$Particles.Initial.RibbonWidth)
   &DynamicMaterialParameters(
    Index 0 Param 1=0.5
    Index 0 Param 2=&MaskFloatBySpawnGroup(Masked Float=0, Passthrough Float=0.2, Spawn Group Mask=1)
    Index 0 Param 3=1
    Module.Param0WriteEnabled=true
    Module.Param1WriteEnabled=true
    Module.Param2WriteEnabled=true
    Module.Param3WriteEnabled=false
   ) {module=DynamicMaterialParameters001}
   &Color(
    Color=&ColorFromCurve(ColorCurve=DataInterface() {propertyValues {
       AlphaCurve {keys [<at 0.0671 1 RCIM_Cubic>,<at 0.7425 0.5>,<at 1.0028 0 RCIM_Cubic>]}
       BlueCurve {keys [<at 0.0295 0.1871 RCIM_Cubic>,<at 0.3103 2>]}
       GreenCurve {keys [<at 0.0295 0 RCIM_Cubic>,<at 0.3103 2>]}
       RedCurve {keys [<at 0.0295 5 RCIM_Cubic>,<at 0.3103 2>]}
      }})
    Scale Alpha=&FNTimeOfDayScalar(
     Night Scale=&Lerp_Float(A=3, B=-1, Alpha=$Particles.Dive)
     Morning Scale=0.2
     Day Scale=&Lerp_Float(A=2, B=2, Alpha=$Particles.Dive)
     Evening Scale=0.5
    )
    Scale Color=<0.2 0.2 0.2>
   )
   &ScaleRibbonWidth(Ribbon Width Scale=&FloatFromCurve(
     FloatCurve=DataInterface() {propertyValues {Curve {keys [<at 0 0.2>,<at 1 1.5>]}}}
     Scale Curve=&Lerp_Float(A=3, B=2, Alpha=$System.FN_Integrator001.Value)
    ))
  ]
  rendererClasses [&NiagaraRibbonRendererProperties]
  renderers [{rendererClass=&NiagaraRibbonRendererProperties}]
 }
 GPU_FlareSmoke001 {
  simTarget=GPUComputeSim
  emitterUpdateScript [
   &EmitterState(Delay First Loop Only=true, Life Cycle Mode=::NewEnumerator1, Loop Delay=0.25, UseLoopDelay=true)
   &SpawnRate(SpawnRate=&Lerp_Float(A=100, B=150, Alpha=$System.FN_Integrator001.NormalizedValue), Use Spawn Probability=false)
  ]
  particleSpawnScript [
   &FN_SetSpawnGroup(Number of Spawn Groups=2)
   &InitializeParticle(
    Color=&UniformRangedLinearColor(Minimum=<rgba 0.1351 0.1423 0.15 1>, Maximum=<rgba 0.117 0.1329 0.15 1>)
    Lifetime=&UniformRangedFloat(Minimum=0.4, Maximum=1)
    Module.Write Color=true
    Module.Write Lifetime=true
    Module.Write Mass=true
    Module.Write Material Random=true
    Module.Write Position=true
    Module.Write Scale=false
    Module.Write SpriteRotation=true
    Module.Write SpriteSize=true
    Position=/Niagara/DynamicInputs/Helpers/SimulationPosition()
    Sprite Rotation=&NormalizedAngleToDegrees(NormalizedAngle=&UniformRangedFloat(Minimum=0, Maximum=360))
    Sprite Size=&Vector2DFromFloat(Value=&RandomRangeFloat(Minimum=200, Maximum=400))
   )
   &SkeletalMeshLocation(
    Apply Sampled Orientation to Mesh Particles=true
    Bone / Socket Index=$Particles.SpawnGroup
    Inherit Velocity=false
    Interpolate Mesh Sampling=false
    MaskSpawnGroup=false
    Mesh Sampling Type=::NewEnumerator5
    Sampled Position Offset=&SelectVectorFromArray(
     Direct Array Index=&SpawnGroup()
     Vector Selection Array=DataInterface() {propertyValues {FloatData [<0 -5 -1>, <0 2 3>]}}
    )
    Sampled Position Offset Coordinate Space:ENiagaraMeshTransforms=::NewEnumerator3
    Skeletal Mesh=$User.SkeletalMesh
    Socket Sampling Mode=::NewEnumerator5
   )
   /SetVariables(Particles.Position=&Add_Vector(
     A=Linked() {linkedVariable {name=Particles.Position, type=NiagaraPosition}}
     B=&RotateVectorByQuaternion(
      Vector To Rotate=&Lerp_Vector(
       A=&Lerp_Vector(A=<-2 0 30>, B=<-7 0 -40>, Alpha=&MakeFloatFromInt(Int=$Particles.SpawnGroup))
       B=&FN_FrontEndVector3Modifier(
        Gameplay=&Lerp_Vector(A=<0 0 40>, B=<0 0 -30>, Alpha=&MakeFloatFromInt(Int=$Particles.SpawnGroup))
        Front End=&Lerp_Vector(A=<0 0 40>, B=<0 0 -30>, Alpha=&MakeFloatFromInt(Int=$Particles.SpawnGroup))
       )
       Alpha=$System.FN_Integrator001.NormalizedValue
      )
      Quaternion=$Output.SkeletalMeshLocation.SampledRotation
     )
    )) {isSetParameters=true}
   /SetVariables(Particles.InitSpriteRotRate=&UniformRangedFloat(Minimum=-3, Maximum=3)) {isSetParameters=true}
   /SetVariables(Particles.IsDiving:NiagaraBool=&SetBoolByFloatComparison(A=$System.FN_Integrator001.NormalizedValue, B=0.01)) {isSetParameters=true}
   /Niagara/Modules/Spawn/Location/SphereLocation(
    MaskSpawnGroup=false
    Module.Override Randomness=false
    Offset=&FN_FrontEndVector3Modifier(
     Gameplay=&Lerp_Vector(A=<0 0 15>, B=<0 0 40>, Alpha=$System.FN_Integrator001.NormalizedValue)
     Front End=&Lerp_Vector(A=<0 0 -10>, B=<0 0 -45>, Alpha=$System.FN_Integrator001.NormalizedValue)
    )
    Sphere Radius=0.1
    Surface Only=false
   ) {module=SphereLocation001}
   &AddVelocityInCone(
    Cone Angle=4
    Cone Axis=&FN_FrontEndVector3Modifier(
     Gameplay=&Lerp_Vector(
      A=&Lerp_Vector(
       A=&RandomRangeVector(Minimum=<0 1 2>, Maximum=<0.4 1.5 2>)
       B=&RandomRangeVector(Minimum=<0 1 2>, Maximum=<0.4 1.5 2>)
       Alpha=&MakeFloatFromInt(Int=$Particles.SpawnGroup)
      )
      B=&Lerp_Vector(A=<0 -0.4 1>, B=<0 -0.4 1>, Alpha=&MakeFloatFromInt(Int=$Particles.SpawnGroup))
      Alpha=&MakeCustomFloatFromBool(Bool=$Particles.IsDiving, True Float=1, False Float=0)
     )
     Front End=<0 0 1>
    )
    Module.Use Velocity Falloff On Cone Axis=false
    Velocity Distribution Along Cone Axis=0
    Velocity Strength=&FN_FrontEndFloatModifier(
     Gameplay=&Lerp_Float(
      A=&RandomRangeFloat(Minimum=400, Maximum=800)
      B=&RandomRangeFloat(Minimum=800, Maximum=1300)
      Alpha=$System.FN_Integrator001.Value
     )
     Front End=&Lerp_Float(A=600, B=800, Alpha=$System.Dive)
    )
   )
  ]
  particleUpdateScript [
   &Drag(Drag=0.5, Use Linear Drag=true, Use Rotational Drag=false)
   &ParticleState()
   &InheritSourceMovement(Applied Position Scale=<0.5 0.5 1>, ApplyForce=false, ApplyPosition=true, ApplyPositionFalloff=false)
   &SpriteRotationRate(Rotation Rate=&Multiply_Float(
     A=$Particles.InitSpriteRotRate
     B=&FloatFromCurve(
      FloatCurve=DataInterface() {propertyValues {Curve {keys [
          {arriveTangent=-1.9214, interpMode=RCIM_Cubic, leaveTangent=-1.9214, tangentMode=RCTM_User, time=-0.0046, value=1.0113}
          {arriveTangent=-0.2496, interpMode=RCIM_Cubic, leaveTangent=-0.2496, tangentMode=RCTM_User, time=0.9896, value=0.16}
         ]}}}
      CurveIndex=$Emitter.Age
      Scale Curve=150
     )
    ))
   &ScaleColor(
    Scale Alpha=&FloatFromCurve(
     FloatCurve=DataInterface() {propertyValues {Curve {keys [
         {interpMode=RCIM_Cubic, tangentMode=RCTM_User, time=0, value=0}
         {arriveTangent=2.125, interpMode=RCIM_Cubic, leaveTangent=2.125, tangentMode=RCTM_User, time=0.3, value=0.8}
         <at 0.4 1>
         {arriveTangent=-0.0528, interpMode=RCIM_Cubic, leaveTangent=-0.0528, tangentMode=RCTM_User, time=0.8, value=0.5}
         {interpMode=RCIM_Cubic, tangentMode=RCTM_Break, time=1, value=0}
        ]}}}
     Scale Curve=&FN_FrontEndFloatModifier(
      Gameplay=&FNTimeOfDayScalar(Night Scale=0.3, Morning Scale=0.5, Day Scale=0.5, Evening Scale=0.3)
      Front End=0.4
     )
    )
    Scale RGB=&VectorFromFloat(Value=&FloatFromCurve(FloatCurve=DataInterface() {propertyValues {Curve {keys [<at 0.1505 1>,<at 1 0>]}}}, Scale Curve=5))
    ScaleA=true
    ScaleRGB=true
   )
   /SetVariables(Particles.ColorForLowEnd:LinearColor=&ColorFromCurve(ColorCurve=DataInterface() {propertyValues {
       AlphaCurve {keys [<at 0 0 RCIM_Cubic>,<at 0.193 0.2>,<at 1 0 RCIM_Cubic>]}
       BlueCurve {keys [<at 0.1178 0 RCIM_Cubic>,<at 0.3202 1>,<at 1 1 RCIM_Cubic>]}
       GreenCurve {keys [<at 0.1178 0.0645 RCIM_Cubic>,<at 0.3202 1>,<at 1 1 RCIM_Cubic>]}
       RedCurve {keys [<at 0.1178 1 RCIM_Cubic>,<at 0.3202 1>,<at 1 1 RCIM_Cubic>]}
      }})) {isSetParameters=true}
   &SolveForcesAndVelocity(Clamp Velocity=false, Limit Acceleration=false)
   &ScaleSpriteSize(Scale Factor=&Vector2DFromFloat(Value=&FloatFromCurve(
      FloatCurve=DataInterface() {propertyValues {Curve {keys [{arriveTangent=7.4766, interpMode=RCIM_Cubic, time=0, value=0.1}, <at 0.5 0.7>, <at 1 2 RCIM_Cubic>]}}}
      Scale Curve=0.65
     )))
   &DynamicMaterialParameters(
    Index 0 Param 1=&Add_Float(
     A=$Particles.RandomFloat
     B=&FloatFromCurve(
      FloatCurve=DataInterface() {propertyValues {Curve {keys [
          {arriveTangent=2.7533, interpMode=RCIM_Cubic, leaveTangent=1.8958, tangentMode=RCTM_User, time=0, value=1}
          {arriveTangent=0.8292, interpMode=RCIM_Cubic, leaveTangent=0.8292, tangentMode=RCTM_User, time=0.321, value=1.3453}
          {arriveTangent=0.0529, interpMode=RCIM_Cubic, leaveTangent=0.2021, tangentMode=RCTM_User, time=1, value=1.5}
         ]}}}
      CurveIndex=&Add_Float(A=&Multiply_Float(A=$Particles.Initial.MaterialRandom, B=0.5), B=$Particles.NormalizedAge)
     )
    )
    Index 0 Param 2=&FloatFromCurve(FloatCurve=DataInterface() {propertyValues {Curve {keys [
         {arriveTangent=0.0453, interpMode=RCIM_Cubic, tangentMode=RCTM_Break, time=0, value=-1}
         {arriveTangent=0.0604, interpMode=RCIM_Cubic, leaveTangent=0.0604, tangentMode=RCTM_User, time=1, value=-1}
        ]}}})
    Index 0 Param 3=&FloatFromCurve(
     FloatCurve=DataInterface() {propertyValues {Curve {keys [
         {arriveTangent=-53.9199, interpMode=RCIM_Cubic, leaveTangent=-8.9727, tangentMode=RCTM_User, time=0, value=1}
         {arriveTangent=0.2347, interpMode=RCIM_Cubic, leaveTangent=0.2723, tangentMode=RCTM_Break, time=0.1, value=0}
        ]}}}
     Scale Curve=0
    )
    Index 0 Param 4=&FloatFromCurve(FloatCurve=DataInterface() {propertyValues {Curve {keys [{interpMode=RCIM_Cubic, leaveTangent=-1.5927, tangentMode=RCTM_Break, time=0, value=1}, <at 1 0>]}}})
    Index 1 Param 1=&FloatFromCurve(
     FloatCurve=DataInterface() {propertyValues {Curve {keys [
         {interpMode=RCIM_Cubic, leaveTangent=4.7088, tangentMode=RCTM_Break, time=0, value=0}
         {arriveTangent=4.2257, interpMode=RCIM_Cubic, leaveTangent=4.2257, tangentMode=RCTM_User, time=0.25, value=1}
        ]}}}
     Scale Curve=&RandomRangeFloat(Minimum=1.5, Maximum=1.5)
    )
    Index 1 Param 2=&FloatFromCurve(
     FloatCurve=DataInterface() {propertyValues {Curve {keys [<at 0 0.2>, {leaveTangent=0.9066, time=1, value=2}]}}}
     Scale Curve=&RandomRangeFloat(Minimum=0.5, Maximum=0.6)
    )
    Index 1 Param 3=&FloatFromCurve(FloatCurve=DataInterface() {propertyValues {Curve {keys [
         {interpMode=RCIM_Cubic, leaveTangent=6.3917, tangentMode=RCTM_Break, time=0, value=0.2}
         {arriveTangent=4.8433, interpMode=RCIM_Cubic, leaveTangent=19.0929, tangentMode=RCTM_User, time=0.25, value=1}
        ]}}})
    Index 1 Param 4=&FloatFromCurve(
     FloatCurve=DataInterface() {propertyValues {Curve {keys [{arriveTangent=-0.3224, interpMode=RCIM_Cubic, leaveTangent=-0.3224, tangentMode=RCTM_User, time=0, value=1}, <at 1 0.9>]}}}
     Scale Curve=&RandomRangeFloat(Minimum=1.1, Maximum=1.4)
    )
    Index 2 Param 1=&FloatFromCurve(FloatCurve=DataInterface() {propertyValues {Curve {keys [
         {interpMode=RCIM_Cubic, leaveTangent=4.2485, tangentMode=RCTM_Break, time=0, value=0}
         {arriveTangent=1.0548, interpMode=RCIM_Cubic, leaveTangent=1.0548, tangentMode=RCTM_User, time=0.17, value=0.5715}
         <at 1 1>
        ]}}})
    Index 2 Param 2=&FloatFromCurve(
     FloatCurve=DataInterface() {propertyValues {Curve {keys [
         {arriveTangent=3.069, interpMode=RCIM_Cubic, leaveTangent=3.069, tangentMode=RCTM_User, time=0, value=0.2}
         {arriveTangent=0.0259, interpMode=RCIM_Cubic, leaveTangent=0.0259, tangentMode=RCTM_User, time=1, value=1.4}
        ]}}}
     Scale Curve=&RandomRangeFloat(Minimum=0.2, Maximum=1.25)
    )
    Index 2 Param 3=&FloatFromCurve(FloatCurve=DataInterface() {propertyValues {Curve {keys [
         {interpMode=RCIM_Cubic, leaveTangent=2.3064, tangentMode=RCTM_Break, time=0, value=1}
         {interpMode=RCIM_Cubic, leaveTangent=11.1233, tangentMode=RCTM_Break, time=1, value=2}
        ]}}})
    Index 2 Param 4=&FloatFromCurve(FloatCurve=DataInterface() {propertyValues {Curve {keys [
         {arriveTangent=-0.6867, interpMode=RCIM_Cubic, leaveTangent=-0.6867, tangentMode=RCTM_User, time=0, value=1}
         {interpMode=RCIM_Cubic, leaveTangent=-1.2956, tangentMode=RCTM_Break, time=1, value=0.5}
        ]}}})
    Index 3 Param 1=&FNTimeOfDayScalar(Night Scale=0.5, Morning Scale=2, Day Scale=2, Evening Scale=0.5)
    Index 3 Param 2=&FNTimeOfDayScalar(Night Scale=160, Morning Scale=150, Day Scale=170, Evening Scale=150)
    Module.Param0WriteEnabled=true
    Module.Param0WriteEnabled1=true
    Module.Param0WriteEnabled2=true
    Module.Param0WriteEnabled3=true
    Module.Param1WriteEnabled=false
    Module.Param1WriteEnabled1=true
    Module.Param1WriteEnabled2=true
    Module.Param1WriteEnabled3=true
    Module.Param2WriteEnabled=true
    Module.Param2WriteEnabled1=true
    Module.Param2WriteEnabled2=true
    Module.Param2WriteEnabled3=true
    Module.Param3WriteEnabled=true
    Module.Param3WriteEnabled1=true
    Module.Param3WriteEnabled2=true
    Module.Param3WriteEnabled3=true
    Write Parameter Index 1=true
    Write Parameter Index 2=true
    Write Parameter Index 3=true
   )
  ]
  rendererClasses [&NiagaraSpriteRendererProperties]
  renderers [{rendererClass=&NiagaraSpriteRendererProperties}]
 }
 GPU_Sparks {
  simTarget=GPUComputeSim
  emitterUpdateScript [
   &EmitterState()
   &SpawnRate(
    Spawn Probability=&Lerp_Float(A=1, B=1, Alpha=$System.FN_Integrator001.Value)
    SpawnRate=&Lerp_Float(A=700, B=2000, Alpha=$System.FN_Integrator001.Value)
    Use Spawn Probability=true
   )
  ]
  particleSpawnScript [
   &FN_SetSpawnGroup(Number of Spawn Groups=2)
   /Niagara/Modules/Spawn/Initialization/V2/InitializeParticle(
    Color=<rgba 40 9.8591 0 1>
    Lifetime Max=&Lerp_Float(
     A=0.2
     B=0.15
     Alpha=&MakeCustomFloatFromBool(Bool=$User.bIsFrontEndPreview, True Float=1, False Float=0)
    )
    Lifetime Min=0.05
    Lifetime Mode=::NewEnumerator1
    Mass Max=0.05
    Mass Min=0.02
    Mass Mode=::NewEnumerator3
    Mesh Scale Mode=::NewEnumerator3
    Mesh Uniform Scale=5
    Sprite Size Max=<1.5 7>
    Sprite Size Min=<1 1>
    Sprite Size Mode=::NewEnumerator2
    UsePositionOffset=false
    Write Material Random=true
   )
   /Game/Effects/Niagara/Modules/Math/FN_EvenNumberToSign(Input:NiagaraInt32=$Particles.SpawnGroup, Output Float Attribute=true)
   /SetVariables(Particles.IsDiving:NiagaraBool=&SetBoolByFloatComparison(A=$System.FN_Integrator001.NormalizedValue, B=0.01)) {isSetParameters=true}
   &SkeletalMeshLocation(
    Apply Sampled Orientation to Mesh Particles=true
    Bone / Socket Index=$Particles.SpawnGroup
    Bone Sampling Mode=::NewEnumerator4
    Inherit Velocity=false
    MaskSpawnGroup=false
    Sampled Position Offset=&Lerp_Vector(
     A=&Multiply_Vector(
      A=&RandomRangeVector(Minimum=<10 0 20>, Maximum=<10 0 30>)
      B=&VectorFromFloat(Value=1)
     )
     B=&RandomRangeVector(Minimum=<-10 0 -15>, Maximum=<-10 0 -20>)
     Alpha=&MakeFloatFromInt(Int=$Particles.SpawnGroup)
    )
    Sampled Position Offset Coordinate Space:ENiagaraMeshTransforms=::NewEnumerator3
    Skeletal Mesh=$User.SkeletalMesh
   )
   &SphereLocation(
    Coordinate Space:ENiagaraCoordinateSpace=::Simulation
    MaskSpawnGroup=false
    Module.Override Randomness=false
    Offset=&Lerp_Vector(A=<0 0 0>, B=<0 0 0>, Alpha=&MakeFloatFromInt(Int=$Particles.SpawnGroup))
    Sphere Radius=2
    Surface Only=false
   )
  ]
  particleUpdateScript [
   &ParticleState()
   &InheritSourceMovement(
    Applied Force Scale=<-0.1 -0.1 0>
    ApplyForce=true
    ApplyForceFalloff=false
    ApplyPosition=true
    ApplyPositionFalloff=false
    Module.LimitForceSpeed=false
   ) {module=InheritSourceMovement001}
   &LinearForce(Force=&VectorFromCurve(
     VectorCurve=DataInterface() {propertyValues {XCurve {keys [<at 0 1>,<at 0.1 0>]}, YCurve {keys [<at 0 1>,<at 0.1 0>]}, ZCurve {keys [<at 0 1>,<at 0.1 0>]}}}
     Scale Curve=&FN_FrontEndVector3Modifier(Gameplay=<0 2000 2000>, Front End=<0 0 2500>)
    ))
   &CurlNoiseForce(
    Curl Noise Cone Mask Angle=90
    Curl Noise Cone Mask Falloff Angle=90
    Mask Curl Noise=true
    Noise Frequency=200
    Noise Strength=100
    Pan Noise=false
   )
   &Drag(Drag=4, Use Linear Drag=true, Use Rotational Drag=false)
   &SolveForcesAndVelocity(Clamp Velocity=false, Limit Acceleration=false)
   &ScaleSpriteSizeBySpeed(
    Max Scale Factor=<1 1>
    Min Scale Factor=<0.5 0.5>
    Sample Scale Factor By Curve=true
    Scale Factor Curve=DataInterface() {propertyValues {Curve {keys [
        {arriveTangent=2.865, interpMode=RCIM_Cubic, time=0.1247, value=-0.0006}
        {arriveTangent=4.1522, interpMode=RCIM_Cubic, leaveTangent=4.1522, tangentMode=RCTM_User, time=0.3948, value=0.4893}
        <at 0.9958 0.9974 RCIM_Cubic>
       ]}}}
    Velocity Threshold=500
   )
   &ScaleSpriteSize(Scale Factor=&Vector2DFromCurve(
     Vector2Curve=DataInterface() {propertyValues {
       XCurve {keys [
         <at 0 0 RCIM_Cubic>
         {arriveTangent=-1.6411, interpMode=RCIM_Cubic, leaveTangent=-1.6411, tangentMode=RCTM_User, time=0.2457, value=1.0046}
         <at 1 0>
        ]}
       YCurve {keys [<at 0 0 RCIM_Cubic>,<at 0.2457 1.0046 RCIM_Cubic>,<at 1 0>]}
      }}
     ScaleCurve=<1 1>
    ))
   &Color(
    Color=&ColorFromCurve(ColorCurve=DataInterface() {propertyValues {
       AlphaCurve {keys [<at -0.0175 0>,<at 0.0353 1 RCIM_Cubic>,<at 1 0 RCIM_Cubic>]}
       BlueCurve {keys [<at 0.0553 0.7794 RCIM_Cubic>,<at 0.4938 0>,<at 1 0 RCIM_Cubic>]}
       GreenCurve {keys [<at 0.0553 0 RCIM_Cubic>,<at 0.4938 0.4491>,<at 1 0 RCIM_Cubic>]}
       RedCurve {keys [<at 0.0553 40 RCIM_Cubic>,<at 0.4938 5>,<at 1 0 RCIM_Cubic>]}
      }})
    Scale Alpha=&FN_FrontEndFloatModifier(
     Gameplay=&FNTimeOfDayScalar(Night Scale=0.5, Morning Scale=11, Day Scale=10, Evening Scale=0.5)
     Front End=1
    )
    Scale Color=&VectorFromFloat(Value=&Lerp_Float(A=20, B=40, Alpha=$System.FN_Integrator001.NormalizedValue))
   )
  ]
  rendererClasses [&NiagaraSpriteRendererProperties]
  renderers [{rendererClass=&NiagaraSpriteRendererProperties}]
 }
}
summary {
 systemName=NS_Contrail_Temple
 emitters [
  {
   emitterName=DistressFlare001
   rendererClasses [
    &NiagaraMeshRendererProperties
    &NiagaraSpriteRendererProperties
    &NiagaraLightRendererProperties
   ]
  }
  {emitterName=FlareRibbon001, rendererClasses [&NiagaraRibbonRendererProperties]}
  {emitterName=GPU_Sparks, simTarget=GPUComputeSim, rendererClasses [&NiagaraSpriteRendererProperties]}
  {emitterName=GPU_FlareSmoke001, simTarget=GPUComputeSim, rendererClasses [&NiagaraSpriteRendererProperties]}
 ]
 userVariables [
  User.bIsFrontEnd {type=NiagaraBool}
  User.bIsFrontEndPreview {type=NiagaraBool}
  User.EmissiveColor {type=LinearColor, description="", defaultValue {struct=LinearColor, value=<rgba 1 1 1 1>}}
  User.FEVec {type=Vector3f, description="", defaultValue {struct=Vector3f, value=<0 0 0>}}
  User.ParaGlideLeanAlpha {type=NiagaraFloat, defaultValue=0, description=""}
  User.PlayerSpeed {type=NiagaraFloat, defaultValue=0, description=""}
  User.RibbonIDOffset {type=NiagaraInt32}
  User.SkeletalMesh {
   type=NiagaraDataInterfaceSkeletalMesh
   defaultValue {
    struct=NiagaraExt_VariableValue_DataInterface
    value {dataInterface=&self:NiagaraDataInterfaceSkeletalMesh_0}
   }
  }
 ]
}
